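# Material function for the SummitReedGrit backpack that outputs an animated, soft-edged angular mask with two opposing wedges rotating over time.
# UVs come from a back projection reused from the WiryPerk glider's MF_ProjectUV, then spin through CustomRotator driven by Time.
# VectorToRadialValue turns them into an angle, which is folded symmetrically through remapped normal and inverted copies.
# Distance from the midpoint feeds an inverted SmoothStep, boosted and saturated for output.
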MaterialFunction /BRCosmetics/Accessories/FORT_Backpacks/Backpack_M_MED_SummitReedGrit/Materials/MF_AnimationMask
graph {nodes [
  n1 = NamedRerouteDeclaration(None=MaterialFunctionCall!(
    Root_Position=Constant3Vector() {constant=<rgba 0 0 0 1>}
    UseClamp?=Constant()
    FlipUV=Constant()
    UVSize=Constant() {r=30}
    CenterOffsetX=Constant()
    CenterOffsetY=Constant()
    CenterOffsetZ=Constant()
   ) {materialFunction=/BRCosmetics/Gadgets/Assets/VinderTech_GliderChute/Glider_WiryPerk/Materials/MF_ProjectUV_WiryPerk}.UV_BackProjection) {name=ProjectUV, nodeColor=<rgba 1 0.0234 0 1>}
  n2 = Multiply(A=Time()) {constB=-0.21}
  n3 = MaterialFunctionCall(Vector or UVs=Add(
    A=MaterialFunctionCall(UVs=@n1, "Rotation Angle (0-1)"=@n2) {materialFunction=&CustomRotator}.Rotated Values
    B=Constant() {r=-0.5}
   )) {materialFunction=&VectorToRadialValue}
  n4 = Saturate(None=LinearInterpolate(Alpha=ComponentMask(None=@n3.Vector Converted to Angle) {r=true}) {constA=-1})
  n5 = OneMinus(None=@n3.Vector Converted to Angle)
  n6 = Saturate(None=LinearInterpolate(Alpha=ComponentMask(None=@n5) {r=true}) {constA=-1})
  n7 = SmoothStep(Min=Constant() {r=0.34}, Max=Constant() {r=0.06}, Value=Distance(A=Max(A=@n4, B=@n6), B=Constant() {r=0.5}))
  FunctionOutput(None=Saturate(None=Multiply(A=@n7) {constB=2}))
 ]}
