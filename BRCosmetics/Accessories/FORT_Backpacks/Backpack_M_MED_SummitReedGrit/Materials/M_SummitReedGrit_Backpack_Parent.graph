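# Parent material for the SummitReedGrit backpack, adding an audio-reactive glow between the 2023 character PreFX and PostFX functions.
# A frontend-loop audio function supplies amplitude, alongside a threshold volume switch and animation-based audio mask.
# FX3 texture channels mask wheel interior, sides and logo, and two hardcoded sphere masks define a red-glow region.
# Masks darken base color and shape emissive, blending two parameterized glow colors with adjustable peak brightness.
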Material /BRCosmetics/Accessories/FORT_Backpacks/Backpack_M_MED_SummitReedGrit/Materials/M_SummitReedGrit_Backpack_Parent
graph {
 properties {
  bUseMaterialAttributes=true
  bUsedWithMeshDeformer=true
  bUsedWithMorphTargets=true
  bUsedWithSkeletalMesh=true
  pixelDepthOffsetMode=PDOM_Legacy
  shadingModel=MSM_FromMaterialExpression
  usedShadingModels=Default Lit | Subsurface | Clear Coat | Subsurface Profile | Cloth
 }
 nodes [
  Constant() {r=1}
  n1 = GetMaterialAttributes(None=NamedRerouteUsage()) {attributeGetTypes [#1, #2]}
  n2 = Multiply(A=@n1.Base Color, B=Constant() {r=0.02})
  n3 = NamedRerouteUsage()
  n4 = MaterialFunctionCall() {materialFunction=&MF_CharacterParent_2023_PreFX}
  MaterialFunctionCall(
   Material=SetMaterialAttributes!(
    MaterialAttributes=@n1.MaterialAttributes
    Base Color=LinearInterpolate(A=@n1.Base Color, B=@n2, Alpha=@n3)
    Emissive Color=NamedRerouteUsage()
   ) {attributeSetTypes [#1, #3]}
   M=@n4.M
   Pre Skinned Local Position=@n4.M
   Distance Blend=@n4.M
   Fresnel=@n4.M
  ) {materialFunction=&MF_CharacterParent_2023_PostFX}
  NamedRerouteDeclaration(None=@n4.Pre skinned Local Position) {name=PreSkinnedLocalPosition, nodeColor=<rgba 0 1 0.8359 1>}
  n5 = MaterialFunctionCall() {materialFunction=&/MF_AnimationMask}
  n6 = NamedRerouteUsage()
  n7 = OneMinus(None=@n6)
  NamedRerouteDeclaration(None=Max(A=Multiply(A=@n5.Result, B=@n7), B=@n6)) {name=AudioMask, nodeColor=<rgba 0.5469 1 0 1>}
  NamedRerouteDeclaration(None=@n4.Result) {name=InputMA, nodeColor=<rgba 1 0.3984 0 1>}
  NamedRerouteDeclaration(None=@n4.M) {name=InputMask, nodeColor=<rgba 0.0781 0 1 1>}
  n8 = NamedRerouteUsage()
  n9 = NamedRerouteUsage()
  n10 = NamedRerouteUsage()
  n11 = Multiply(
   A=LinearInterpolate(B=ScalarParameter() {defaultValue=2.5, group=SummitReedReactivity, parameterName=BrightestGlow}, Alpha=Multiply(A=@n8, B=@n9)) {constA=1, constB=2}
   B=LinearInterpolate(
    A=VectorParameter() {group=SummitReedReactivity, parameterName=GlowColorA, defaultValue=<rgba 0.05 0.02 0.02 1>}.RGB
    B=VectorParameter() {group=SummitReedReactivity, parameterName=GlowColorB, defaultValue=<rgba 1 0.01 0.01 1>}.RGB
    Alpha=@n10
   )
  )
  n12 = LinearInterpolate(Alpha=@n10) {constA=0.01, constB=2}
  NamedRerouteDeclaration(None=Multiply(A=NamedRerouteUsage(), B=Multiply(A=@n11, B=@n12))) {name=OutEmissiveColor, nodeColor=<rgba 0.2422 1 0 1>}
  NamedRerouteDeclaration(None=Power(Base=MaterialFunctionCall!(InAmplitude, InSpectrum, CustomFakeSpectrum, CustomFakeAmplitude) {materialFunction=/BRCosmetics/Global/MaterialFunctions/AudioReactivity/MF_AudioReactivity_FrontendLoop}.Amplitude)) {name=Amplitude, nodeColor=<rgba 1 0.5391 0 1>}
  n13 = NamedRerouteUsage()
  n14 = ComponentMask(None=Constant4Vector() {constant=<rgba 0 -9.9099 0.424 1>}.RGB) {b=true, g=true, r=true}
  n15 = Constant() {r=13.9}
  n16 = Constant() {r=0.98}
  n17 = NamedRerouteUsage()
  n18 = ComponentMask(None=Constant4Vector() {constant=<rgba 0 -2.6975 -0.224 1>}.RGB) {b=true, g=true, r=true}
  n19 = Constant() {r=14.3}
  n20 = Constant() {r=0.98}
  NamedRerouteDeclaration(None=Max(A=SphereMask(A=@n13, B=@n14, Radius=@n15, Hardness=@n16), B=SphereMask(A=@n17, B=@n18, Radius=@n19, Hardness=@n20))) {name=RedGlowingMask, nodeColor=<rgba 0 1 0.9063 1>}
  n21 = NamedRerouteUsage()
  n22 = ComponentMask(None=NamedRerouteUsage()) {r=true}
  n23 = OneMinus(None=@n22)
  n24 = Multiply(A=Power(Base=Saturate(None=@n23)) {constExponent=1}) {constB=0.8}
  n25 = Multiply(A=NamedRerouteUsage()) {constB=0.2}
  n26 = Constant() {r=15}
  n27 = LinearInterpolate(
   A=Saturate(None=Power(Base=@n22, Exp=@n26) {constExponent=10})
   B=Multiply(A=OneMinus(None=@n21), B=Max(A=@n24, B=@n25))
   Alpha=NamedRerouteUsage()
  )
  n28 = OneMinus(None=NamedRerouteUsage())
  NamedRerouteDeclaration(None=Multiply(A=Multiply(A=@n27, B=@n28), B=NamedRerouteUsage())) {name=FinalMask, nodeColor=<rgba 0 0.5547 1 1>}
  n29 = TextureSample() {
   samplerType=SAMPLERTYPE_LinearColor
   texture=/BRCosmetics/Accessories/FORT_Backpacks/Backpack_M_MED_SummitReedGrit/Textures/T_SummitReedGrit_Backpack_FX3
  }
  NamedRerouteDeclaration(None=@n29.R) {name=WheelInside, nodeColor=<rgba 1 0.8203 0 1>}
  NamedRerouteDeclaration(None=@n29.G) {name=SidesMask, nodeColor=<rgba 1 0 0.4687 1>}
  NamedRerouteDeclaration(None=@n29.B) {name=LogoMask, nodeColor=<rgba 0.9688 0 1 1>}
  n30 = Constant() {r=1}
  NamedRerouteDeclaration(None=If(A=NamedRerouteUsage(), "A > B"=@n30, "A == B"=@n30, "A < B"=Constant()) {constB=0.02}) {name=VolumeControSwitch, nodeColor=<rgba 0.7109 0 1 1>}
 ]
}
